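# Additive, unlit, two-sided material that fakes a light glow on a box mesh, with separate day and night colors.
# Height within instance bounds, raised to Falloff, scales the time-of-day color and Intensity through the emissive adjustment function, brightest at the top.
# Two Distance_Blend calls combine with Min to fade the glow when the camera is near or far.
# A horizontal normal-based vertex offset flares the box toward its bottom by TaperWidth.
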Material /WildEstate/Environment/Props/LightSource/BR_FakeLight_Box/M_FakeLightBox
graph {
 properties {bAutomaticallySetUsageInEditor=false, blendMode=BLEND_Additive, shadingModel=MSM_Unlit, twoSided=true, usedShadingModels=Unlit}
 nodes [
  n1 = Power(
   Base=MaterialFunctionCall() {materialFunction=/Game/Effects/Niagara/MasterMaterials/Functions/InstanceBounds_0-1_UVW}.B
   Exp=ScalarParameter() {defaultValue=5, group=Light, parameterName=Falloff, sortPriority=3}
  ) {constExponent=5}
  n2 = VectorParameter() {group=Light, parameterName=Color Day, sortPriority=1, defaultValue=<rgba 1 0.8228 0.7011 1>}
  n3 = VectorParameter() {group=Light, parameterName=Color Night, sortPriority=2, defaultValue=<rgba 1 0.8228 0.7011 1>}
  n4 = MaterialFunctionCall(Color_Morning=@n2.RGB, Color_Day=@n2.RGB, Color_Evening=@n3.RGB, Color_Night=@n3.RGB) {materialFunction=&Color_Time_of_Day}
  MaterialFunctionCall(Emissive=Multiply(A=Multiply(A=@n1, B=@n4.Result), B=ScalarParameter() {defaultValue=0.1, group=Light, parameterName=Intensity, sortPriority=0})) {materialFunction=&MF_FX_EmissiveAdjustments}
  n5 = ScalarParameter() {
   defaultValue=1000
   desc=Distance at which light starts to disappear when get nearby.
   group=Opacity
   parameterName=NearFadeDistance
   sortPriority=0
  }
  n6 = ScalarParameter() {
   defaultValue=2000
   desc=Distance at which light starts to disappear when when get away from it.
   group=Opacity
   parameterName=FarFadeDistance
   sortPriority=1
  }
  n7 = Multiply(A=@n6) {constB=-1}
  Min(
   A=MaterialFunctionCall!(Blend Range=@n5, Start Offset=Constant()) {materialFunction=/Engine/Functions/Engine_MaterialFunctions03/Particles/Distance_Blend}.Result
   B=OneMinus(None=MaterialFunctionCall(Blend Range=@n6, Start Offset=@n7) {materialFunction=&Distance_Blend}.Result)
  )
  n8 = Multiply(A=VertexNormalWS(), B=Constant3Vector() {constant=<rgba 1 1 0 0>})
  n9 = MaterialFunctionCall() {materialFunction=&InstanceBounds_0-1_UVW}
  Multiply(
   A=Multiply(A=Normalize(VectorInput=@n8), B=OneMinus(None=@n9.B))
   B=ScalarParameter() {defaultValue=100, group=Shape, parameterName=TaperWidth, sortPriority=0}
  )
 ]
}
